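# Physics asset for the left-mirror damage piece of the dynamic SUV vehicle mesh.
# A large capsule covers the body bone and a small capsule wraps the mirror_l bone.
# A single constraint joins mirror_l to body with limited swing and twist, so the mirror can swing within a bounded cone relative to the body.
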PhysicsAsset /BasicSUV/Meshes/SUV_Veh_Dyn_dmg_mirror_l
bodies {
 body [{shapeType=Capsule, length=229.3724, radius=125.9405, center=<40.2655 -0.1455 66.4146>, rotation=<rot 0 -89.8532 -87.571>}]
 mirror_l [{shapeType=Capsule, length=18, radius=9, shapeName=mirror_l_capsule, center=<-1.9672 -10.8121 3.567>, rotation=<rot 0 -12.3788 -72.8181>}]
}
constraints [{bone1Name=mirror_l, bone2Name=body, swing1LimitDegrees=35, swing2LimitDegrees=40, twistLimitDegrees=35}]
