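# Physics asset for the SUV's trunk damage mesh, giving a simplified collision and joint setup to the vehicle body and trunk.
# One large capsule approximates the body, and a smaller, narrower capsule covers the trunk.
# A single constraint joins trunk to body, with equal swing and twist limit angles, so the trunk can be simulated as a jointed part.
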PhysicsAsset /BasicSUV/Meshes/SUV_Veh_Dyn_dmg_trunk
bodies {
 body [{shapeType=Capsule, length=229.3724, radius=125.9405, center=<40.2655 -0.1455 66.4146>, rotation=<rot 0 -89.8532 -87.571>}]
 trunk [{shapeType=Capsule, length=82.633, radius=20, center=<-3.571 55.8556 -0.1296>, rotation=<rot 0 0 90>}]
}
constraints [{bone1Name=trunk, bone2Name=body, swing1LimitDegrees=45, swing2LimitDegrees=45, twistLimitDegrees=45}]
