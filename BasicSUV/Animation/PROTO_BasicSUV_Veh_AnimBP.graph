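# Animation Blueprint for the BasicSUV vehicle that poses its axles, wheels, steering wheel and chassis.
# From mesh ref pose, chained bone transforms apply axle, steering and spin variables, then tire-mod offsets and wheel-well scaling, ending in a bouncy chassis controller.
# The vehicle's mod-applied client delegate sets bHasTireMod true for Offroad tires, false for Bulletproof.
# Trunk and mirror damage nodes, including RigidBody simulations, are not connected to the output pose.
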AnimBlueprint /BasicSUV/Animation/PROTO_BasicSUV_Veh_AnimBP
target_skeleton=/BasicSUV/Meshes/SUV_Veh_Skeleton
graphs [
 AnimGraph {nodes [
   Local To Component(LocalPose=Local Space Ref Pose())
   n1 = Get TrunkDamageAlpha() -> (TrunkDamageAlpha:float)
   "Transform (Modify) Bone - Bone: trunk"(Alpha=@n1, Translation="0,0,0", Rotation="0,11,0", Scale="1,1,1") {
    rotationMode=BMM_Additive
    rotationSpace=BCS_BoneSpace
    alphaBoolBlend {blendOption=Linear}
    alphaScaleBiasClamp {bInterpResult=true, interpSpeedDecreasing=5, interpSpeedIncreasing=5}
    boneToModify {boneName=trunk}
    rotation {yaw=11}
   }
   n2 = RigidBody(Alpha=Get MirrorLeftDamageAlpha() -> (MirrorLeftDamageAlpha:float)) {
    bOverrideWorldGravity=true
    evaluationResetTime=-1
    overridePhysicsAsset=/BasicSUV/Meshes/SUV_Veh_Dyn_dmg_mirror_l
    alphaBoolBlend {blendOption=Linear}
    baseBoneRef {boneName=body}
    overrideWorldGravity {z=-980}
    simSpaceSettings {
     maxAngularAcceleration=2000
     maxAngularVelocity=2000
     maxLinearAcceleration=2000
     maxLinearVelocity=2000
     worldAlpha=1
     externalLinearDragV=<1 1 1>
    }
   }
   RigidBody(ComponentPose=@n2, Alpha=Get MirrorRightDamageAlpha() -> (MirrorRightDamageAlpha:float)) {
    bOverrideWorldGravity=true
    evaluationResetTime=-1
    overridePhysicsAsset=/BasicSUV/Meshes/SUV_Veh_Dyn_dmg_mirror_r
    alphaBoolBlend {blendOption=Linear}
    baseBoneRef {boneName=body}
    overrideWorldGravity {z=-980}
    simSpaceSettings {
     maxAngularAcceleration=2000
     maxAngularVelocity=2000
     maxLinearAcceleration=2000
     maxLinearVelocity=2000
     worldAlpha=1
     externalLinearDragV=<1 1 1>
    }
   }
   RigidBody(Alpha=@n1) {
    bOverrideWorldGravity=true
    evaluationResetTime=-1
    overridePhysicsAsset=/BasicSUV/Meshes/SUV_Veh_Dyn_dmg_trunk
    alphaBoolBlend {blendOption=Linear}
    baseBoneRef {boneName=body}
    componentLinearAccScale {z=2}
    componentLinearVelScale {z=2}
    overrideWorldGravity {z=-980}
    simSpaceSettings {
     maxAngularAcceleration=2000
     maxAngularVelocity=2000
     maxLinearAcceleration=2000
     maxLinearVelocity=2000
     worldAlpha=1
     externalLinearDragV=<1 1 1>
    }
   }
   n3 = "Transform (Modify) Bone - Bone: front_differential"(
    ComponentPose=Mesh Space Ref Pose!() -> (ComponentPose:Component Space Pose Link Structure) {is=title AnimGraphNode_MeshRefPose}
    Alpha=1
    Translation=Get AxlePositionF() -> (AxlePositionF:Vector)
    Rotation=Get AxleRotationF() -> (AxleRotationF:Rotator)
    Scale="1,1,1"
   ) {
    rotationMode=BMM_Replace
    rotationSpace=BCS_BoneSpace
    translationMode=BMM_Replace
    translationSpace=BCS_BoneSpace
    alphaBoolBlend {blendOption=Linear}
    boneToModify {boneName=front_differential}
    translation {z=-500}
   }
   n4 = "Transform (Modify) Bone - Bone: rear_differential"(
    ComponentPose=@n3
    Alpha=1
    Translation=Get AxlePositionB() -> (AxlePositionB:Vector)
    Rotation=Get AxleRotationB() -> (AxleRotationB:Rotator)
    Scale="1,1,1"
   ) {
    rotationMode=BMM_Replace
    rotationSpace=BCS_BoneSpace
    translationMode=BMM_Replace
    translationSpace=BCS_BoneSpace
    alphaBoolBlend {blendOption=Linear}
    boneToModify {boneName=rear_differential}
   }
   n5 = "Transform (Modify) Bone - Bone: wheel_steering_fr_r"(ComponentPose=@n4, Alpha=1, Translation="0,0,0", Rotation=Get WheelRotationFR() -> (WheelRotationFR:Rotator), Scale="1,1,1") {
    rotationMode=BMM_Replace
    rotationSpace=BCS_BoneSpace
    translationSpace=BCS_BoneSpace
    alphaBoolBlend {blendOption=Linear}
    boneToModify {boneName=wheel_steering_fr_r}
   }
   n6 = "Transform (Modify) Bone - Bone: wheel_steering_fr_l"(ComponentPose=@n5, Alpha=1, Translation="0,0,0", Rotation=Get WheelRotationFL() -> (WheelRotationFL:Rotator), Scale="1,1,1") {
    rotationMode=BMM_Replace
    rotationSpace=BCS_BoneSpace
    translationSpace=BCS_BoneSpace
    alphaBoolBlend {blendOption=Linear}
    boneToModify {boneName=wheel_steering_fr_l}
   }
   n7 = "Transform (Modify) Bone - Bone: wheel_tilted_fr_r"(ComponentPose=@n6, Alpha=1, Translation="0,0,0", Rotation=Get WheelSpinFR() -> (WheelSpinFR:Rotator), Scale="1,1,1") {rotationMode=BMM_Replace, rotationSpace=BCS_BoneSpace, alphaBoolBlend {blendOption=Linear}, boneToModify {boneName=wheel_tilted_fr_r}}
   n8 = "Transform (Modify) Bone - Bone: wheel_tilted_fr_l"(ComponentPose=@n7, Alpha=1, Translation="0,0,0", Rotation=Get WheelSpinFL() -> (WheelSpinFL:Rotator), Scale="1,1,1") {rotationMode=BMM_Replace, rotationSpace=BCS_BoneSpace, alphaBoolBlend {blendOption=Linear}, boneToModify {boneName=wheel_tilted_fr_l}}
   n9 = "Transform (Modify) Bone - Bone: wheel_tilted_bk_r"(ComponentPose=@n8, Alpha=1, Translation="0,0,0", Rotation=Get WheelSpinBR() -> (WheelSpinBR:Rotator), Scale="1,1,1") {rotationMode=BMM_Replace, rotationSpace=BCS_BoneSpace, alphaBoolBlend {blendOption=Linear}, boneToModify {boneName=wheel_tilted_bk_r}}
   n10 = "Transform (Modify) Bone - Bone: wheel_tilted_bk_l"(ComponentPose=@n9, Alpha=1, Translation="0,0,0", Rotation=Get WheelSpinBL() -> (WheelSpinBL:Rotator), Scale="1,1,1") {rotationMode=BMM_Replace, rotationSpace=BCS_BoneSpace, alphaBoolBlend {blendOption=Linear}, boneToModify {boneName=wheel_tilted_bk_l}}
   n11 = "Transform (Modify) Bone - Bone: steering_wheel"(ComponentPose=@n10, Alpha=1, Translation="0,0,0", Rotation=Get SteeringRotation() -> (SteeringRotation:Rotator), Scale="1,1,1") {rotationMode=BMM_Replace, rotationSpace=BCS_BoneSpace, alphaBoolBlend {blendOption=Linear}, boneToModify {boneName=steering_wheel}}
   n12 = Get bHasTireMod() -> (bHasTireMod:bool)
   n13 = "Transform (Modify) Bone - Bone: wheel_steering_bk_l"!(ComponentPose=@n11, bAlphaBoolEnabled:bool=@n12, Translation="0,-10,-3", Rotation="0,0,0", Scale="1,1,1") -> (Pose) {
    alphaInputType=Bool
    rotationSpace=BCS_BoneSpace
    scaleMode=BMM_Additive
    scaleSpace=BCS_BoneSpace
    translationMode=BMM_Additive
    translationSpace=BCS_BoneSpace
    alphaBoolBlend {blendOption=Linear}
    boneToModify {boneName=wheel_steering_bk_l}
    translation {y=-10, z=-3}
   }
   n14 = "Transform (Modify) Bone - Bone: wheel_steering_bk_r"!(ComponentPose=@n13, bAlphaBoolEnabled:bool=@n12, Translation="0,10,-3", Rotation="0,0,0", Scale="1,1,1") -> (Pose) {
    alphaInputType=Bool
    rotationSpace=BCS_BoneSpace
    scaleMode=BMM_Additive
    scaleSpace=BCS_BoneSpace
    translationMode=BMM_Additive
    translationSpace=BCS_BoneSpace
    alphaBoolBlend {blendOption=Linear}
    boneToModify {boneName=wheel_steering_bk_r}
    translation {y=10, z=-3}
   }
   n15 = "Transform (Modify) Bone - Bone: steering_knuckle_fr_l"!(ComponentPose=@n14, bAlphaBoolEnabled:bool=@n12, Translation="0,-10,0", Rotation="0,0,0", Scale="1,1,1") -> (Pose) {
    alphaInputType=Bool
    rotationSpace=BCS_BoneSpace
    translationMode=BMM_Additive
    translationSpace=BCS_BoneSpace
    alphaBoolBlend {blendOption=Linear}
    boneToModify {boneName=steering_knuckle_fr_l}
    translation {y=-10}
   }
   n16 = "Transform (Modify) Bone - Bone: steering_knuckle_fr_r"!(ComponentPose=@n15, bAlphaBoolEnabled:bool=@n12, Translation="0,-10,0", Rotation="0,0,0", Scale="1,1,1") -> (Pose) {
    alphaInputType=Bool
    rotationSpace=BCS_BoneSpace
    translationMode=BMM_Additive
    translationSpace=BCS_BoneSpace
    alphaBoolBlend {blendOption=Linear}
    boneToModify {boneName=steering_knuckle_fr_r}
    translation {y=-10}
   }
   n17 = Get bHasTireMod()
   n18 = "Transform (Modify) Bone - Bone: wheel_well_spreader_bk_l"!(ComponentPose=@n16, bAlphaBoolEnabled:bool=@n17, Translation="6,0,0", Rotation="0,0,0", Scale="1.4,1,1") -> (Pose) {
    alphaInputType=Bool
    rotationSpace=BCS_BoneSpace
    scaleMode=BMM_Additive
    scaleSpace=BCS_BoneSpace
    translationMode=BMM_Additive
    translationSpace=BCS_BoneSpace
    alphaBoolBlend {blendOption=Linear}
    alphaScaleBiasClamp {bInterpResult=true, interpSpeedDecreasing=5, interpSpeedIncreasing=5}
    boneToModify {boneName=wheel_well_spreader_bk_l}
    scale {x=1.4}
    translation {x=6}
   }
   n19 = "Transform (Modify) Bone - Bone: wheel_well_spreader_bk_r"!(ComponentPose=@n18, bAlphaBoolEnabled:bool=@n17, Translation="6,0,0", Rotation="0,0,0", Scale="1.4,1,1") -> (Pose) {
    alphaInputType=Bool
    rotationSpace=BCS_BoneSpace
    scaleMode=BMM_Additive
    scaleSpace=BCS_BoneSpace
    translationMode=BMM_Additive
    translationSpace=BCS_BoneSpace
    alphaBoolBlend {blendOption=Linear}
    alphaScaleBiasClamp {bInterpResult=true, interpSpeedDecreasing=5, interpSpeedIncreasing=5}
    boneToModify {boneName=wheel_well_spreader_bk_r}
    scale {x=1.4}
    translation {x=6}
   }
   n20 = "Transform (Modify) Bone - Bone: wheel_well_spreader_fr_l"!(ComponentPose=@n19, bAlphaBoolEnabled:bool=@n17, Translation="1,0,0", Rotation="0,0,0", Scale="1.6,1,1") -> (Pose) {
    alphaInputType=Bool
    rotationSpace=BCS_BoneSpace
    scaleMode=BMM_Additive
    scaleSpace=BCS_BoneSpace
    translationMode=BMM_Additive
    translationSpace=BCS_BoneSpace
    alphaBoolBlend {blendOption=Linear}
    alphaScaleBiasClamp {bInterpResult=true, interpSpeedDecreasing=5, interpSpeedIncreasing=5}
    boneToModify {boneName=wheel_well_spreader_fr_l}
    scale {x=1.6}
    translation {x=1}
   }
   n21 = "Transform (Modify) Bone - Bone: wheel_well_spreader_fr_r"!(ComponentPose=@n20, bAlphaBoolEnabled:bool=@n17, Translation="1,0,0", Rotation="0,0,0", Scale="1.6,1,1") -> (Pose) {
    alphaInputType=Bool
    rotationSpace=BCS_BoneSpace
    scaleMode=BMM_Additive
    scaleSpace=BCS_BoneSpace
    translationMode=BMM_Additive
    translationSpace=BCS_BoneSpace
    alphaBoolBlend {blendOption=Linear}
    alphaScaleBiasClamp {bInterpResult=true, interpSpeedDecreasing=5, interpSpeedIncreasing=5}
    boneToModify {boneName=wheel_well_spreader_fr_r}
    scale {x=1.6}
    translation {x=1}
   }
   AnimGraphNode_BouncyChassisVehicleController_1 = Vehicle Bouncy Chassis Controller!(
    ComponentPose:Component Space Pose Link Structure=@n21
    Alpha:float=1
    BouncyChassis:Bouncy Chassis Vehicle State Structure=(PitchDegrees=0,RollDegrees=0,TranslationZ=0)
   ) -> (Pose:Component Space Pose Link Structure) {is=title AnimGraphNode_VehicleBouncyChassisController, alphaBoolBlend {blendOption=Linear}}
   Output Pose(Result=Component To Local(ComponentPose=@AnimGraphNode_BouncyChassisVehicleController_1)) {layerGroup=None}
  ]}
 EventGraph {nodes [
   n1 = OnModAppliedOnClientDelegate_Event!() -> (OutputDelegate, then, VehicleActor:Actor Object Reference, ModTag:"Gameplay Tag Structure (by ref)") {is=title K2Node_CustomEvent}
   n2 = "Equal (GameplayTag)"(A=@n1.ModTag, B=(TagName="Vehicle.Mod.Tire.Offroad"))
   > n3 = Branch(Condition=@n2)
   > Set bHasTireMod(bHasTireMod:bool=@n2) -> (Output_Get:bool)
   > PrintString!(execute, InString=OffRoad Tire Applied, bPrintToScreen, bPrintToLog, TextColor, Duration, Key) -> (then)
   Branch(execute=@n3.else, Condition="Equal (GameplayTag)"(A=@n1.ModTag, B=(TagName="Vehicle.Mod.Tire.Bulletproof")))
   > Set bHasTireMod(bHasTireMod=false)
   > PrintString(InString=Bulletproof Tire Applied)
   Event BlueprintInitializeAnimation()
   n4 = Get Vehicle() -> (Vehicle:Fort Athena SKVehicle Object Reference)
   > Is Valid(InputObject=@n4)
   > Assign On Mod Applied on Client Delegate!(execute:Exec, self:Fort Athena Vehicle Object Reference=@n4, Delegate:"Delegate (by ref)"=@n1) -> (then:Exec) {is=title K2Node_AssignDelegate}
  ]}
]
variables [
 bHasTireMod {type=boolean}
 MirrorLeftDamageAlpha {type=number}
 MirrorRightDamageAlpha {type=number}
 TrunkDamageAlpha {type=number}
]
