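# Sound cue for the BasicSUV engine-off sound, built on a mono shutoff recording.
# A single wave player feeds the output directly, with no randomization, mixing or modulation nodes.
# The cue boosts volume and uses the shared vehicle movement attenuation and engine concurrency settings.
# It is assigned to the Fort_Vehicles_Interior sound class.
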SoundCue /BasicSUV/Audio/Veh_BasicSUV_Engine_Off_Cue
root=WavePlayer_0
nodes [{id=WavePlayer_0, properties {soundWaveAssetPtr=&/MERC_C63_Shutoff_Mono}}]
properties {
 attenuationSettings=/VehicleAudio/Settings/Vehicle_Movement_Att
 maxDistance=12512
 soundClassObject=&Fort_Vehicles_Interior
 volumeMultiplier=3
 concurrencySet [/VehicleAudio/Settings/Vehicle_Engine_Conc]
}
