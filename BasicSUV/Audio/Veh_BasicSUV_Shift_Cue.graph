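# Sound cue for the BasicSUV gear shift, pairing a mechanical shift sound with an exhaust pop.
# A mixer combines two modulated random branches: one of three gear shift waves, and one of four MERC C63 exhaust pops.
# The pop branch has a short fixed delay and a boosted mix level, so it follows the shift.
# Output uses the vehicle interior sound class and close-detail movement attenuation.
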SoundCue /BasicSUV/Audio/Veh_BasicSUV_Shift_Cue
root=Mixer_0
nodes [
 {id=Delay_0, children [Modulator_1], properties {delayMax=0.15, delayMin=0.15}}
 {id=Mixer_0, children [Modulator_0,Delay_0], properties {inputVolume [1,3.2]}}
 {id=Modulator_0, children [Random_1]}
 {id=Modulator_1, children [Random_0]}
 {id=Random_0, children [WavePlayer_7, WavePlayer_8, WavePlayer_9, WavePlayer_10], properties {weights [1,1,1,1]}}
 {id=Random_1, children [WavePlayer_2,WavePlayer_3,WavePlayer_4], properties {weights [1,1,1]}}
 {id=WavePlayer_10, properties {soundWaveAssetPtr=&/MERC_C63_ExhaustPop_04}}
 {id=WavePlayer_2, properties {soundWaveAssetPtr=/VehicleAudio/Assets/Gears/Veh_Valet_Gen_Gear_Shift_Mech_03}}
 {id=WavePlayer_3, properties {soundWaveAssetPtr=/VehicleAudio/Assets/Gears/Veh_Valet_Gen_Gear_Shift_Mech_01}}
 {id=WavePlayer_4, properties {soundWaveAssetPtr=/VehicleAudio/Assets/Gears/Veh_Valet_Gen_Gear_Shift_Mech_02}}
 {id=WavePlayer_7, properties {soundWaveAssetPtr=&/MERC_C63_ExhaustPop_01}}
 {id=WavePlayer_8, properties {soundWaveAssetPtr=&/MERC_C63_ExhaustPop_02}}
 {id=WavePlayer_9, properties {soundWaveAssetPtr=&/MERC_C63_ExhaustPop_03}}
]
properties {
 attenuationSettings=/VehicleAudio/Settings/Vehicle_Movement_Close_Detail_Att
 maxDistance=12000
 soundClassObject=&Fort_Vehicles_Interior
 volumeMultiplier=0.5
}
