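# Physics asset for the BasicSUV vehicle skeletal mesh, giving the SUV its collision shape.
# It holds a single body on the frame bone with one box shape named Trunk, offset toward the negative X end of the vehicle and raised above the bone origin.
# There are no constraints, so no other bones carry simulated bodies or joints.
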PhysicsAsset /BasicSUV/Meshes/SUV_Veh_Physics
bodies {frame [{
   shapeType=Box
   extentX=83.3125
   extentY=178.1875
   extentZ=85.5625
   shapeName=Trunk
   center=<-121.6485 0 141.0527>
   rotation=<rot 0 0 -0.0001>
  }]}
constraints []
