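# Physics asset for the BasicSUV vehicle, named for fluid simulation, that gives the SUV a simplified collision shape.
# A single body on the frame bone combines two stacked boxes for the lower hull and shorter, narrower cabin with two low sideways capsules toward front and rear, roughly spanning the axles.
# It has no other bodies or constraints, so wheels and other parts get no separate physics.
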PhysicsAsset /BasicSUV/Meshes/SUV_Veh_Physics_Fluidsim
bodies {frame [
  {shapeType=Capsule, length=169.5, radius=40.5, shapeName=frame_capsule, center=<-97.4434 0 21.033>, rotation=<rot 0 0 90>}
  {shapeType=Capsule, length=169.5, radius=40.5, shapeName=frame_capsule, center=<166.3649 0 21.033>, rotation=<rot 0 0 90>}
  {shapeType=Box, extentX=399.5, extentY=193.5, extentZ=104.5, shapeName=frame_box, center=<19.6309 0 71.4348>}
  {shapeType=Box, extentX=247.5, extentY=174.5, extentZ=104.5, shapeName=frame_box, center=<-30.8109 0 137.213>}
 ]}
constraints []
